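# Sound cue for collecting a Fortbyte quest item in Athena.
# A Random node picks one of two equally weighted collect waves and feeds it through a Modulator at the root.
# It plays in the non-critical in-game UI sound class with quest-interactable hint attenuation and a boosted volume.
# Concurrency allows a single instance, and new plays are blocked while one is active.
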SoundCue /Game/Athena/Sounds/Quests/Fortbyte/Quest_Fortbyte_Collect_Cue
root=Modulator_0
nodes [
 {id=Modulator_0, children [Random_0]}
 {id=Random_0, children [WavePlayer_0,WavePlayer_2], properties {weights [1,1]}}
 {id=WavePlayer_0, properties {soundWaveAssetPtr=&/Fortbyte_Collect_01}}
 {id=WavePlayer_2, properties {soundWaveAssetPtr=&/Fortbyte_Collect_02}}
]
properties {
 attenuationSettings=/Game/Sounds/Attenuation/Fort_QuestInteractables_Hint
 maxDistance=1300
 soundClassObject=&Fort_UI_InGame_NonCritical
 volumeMultiplier=1.45
 concurrencyOverrides {maxCount=1, resolutionRule=PreventNew}
}
